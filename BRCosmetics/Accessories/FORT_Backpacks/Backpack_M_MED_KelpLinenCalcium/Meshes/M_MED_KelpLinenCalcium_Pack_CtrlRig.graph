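# Control Rig for the KelpLinenCalcium backpack that poses the pack bone and oscillates its ring of strike bones.
# Setup caches an offset pack position and fills per-strike phase and seeded random speed arrays.
# Each update blends the pack toward a tilted pose by remapped Spine_rotation, then applies the cached offset.
# It also slides each strike along local X using a sine of time_from_start with those phases and speeds.
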ControlRigBlueprint /BRCosmetics/Accessories/FORT_Backpacks/Backpack_M_MED_KelpLinenCalcium/Meshes/M_MED_KelpLinenCalcium_Pack_CtrlRig
graphs [
 RigVMFunctionLibrary {links [], nodes []}
 RigVMFunction_Sequence_1_ContainedGraph {nodes [
   n1 = RigVMFunctionEntryNode()
   RigVMFunction_Sequence = &RigVMFunction_Sequence(ExecuteContext=@n1.ExecuteContext)
   n2 = &RigVMFunction_Sequence(ExecuteContext=@RigVMFunction_Sequence.B)
   RigVMFunctionReturnNode(A=@RigVMFunction_Sequence.A, B=@n2.A, C=@n2.B)
  ]}
 RigVMModel {nodes [
   n1 = &RigUnit_PrepareForExecution()
   RigUnit_GetTransform = &RigUnit_GetTransform(Item=(Type=Bone,Name="pack"), bInitial=False)
   RigVMVariableNode(
    Value=/Script/RigVM.RigVMFunction_MathVectorAdd(B=(X=0,Y=-1.5,Z=0), A=@RigUnit_GetTransform.Transform.Translation) {is=struct RigVMUnitNode, template="Add::Execute(in A,in B,out Result)"}.Result
    ExecuteContext=@n1.ExecutePin
   ) {variable=pack_start_pos}
   > RigVMVariableNode(Value=1) {variable=strikes_move_dist}
   > RigVMVariableNode(Value=1.7) {variable=strikes_move_speed}
   > RigVMVariableNode(Value=0.7) {variable=strikes_pos_offset}
   > RigVMFunction_ForLoopCount = /Script/RigVM.RigVMFunction_ForLoopCount(Count=14) {is=struct RigVMUnitNode}
   RigVMFunction_MathIntToDouble_1 = /Script/RigVM.RigVMFunction_MathIntToFloat(Value=@RigVMFunction_ForLoopCount.Index) {is=struct RigVMUnitNode, template="Cast::Execute(in Value,out Result)"}
   Multiply = &RigVMFunction_MathDoubleMul(B=1.2, A=@RigVMFunction_MathIntToDouble_1.Result)
   > Add = RigVMDispatchNode(Element=@Multiply.Result, Array=RigVMVariableNode() {variable=strikes_moving_offset}.Value) {template="DISPATCH_RigVMDispatch_ArrayAdd(io Array,in Element,out Index)"}
   Random = /Script/RigVM.RigVMFunction_RandomFloat(Duration=0, Maximum=2, Minimum=0.5, Seed=@Add.Index) {is=struct RigVMUnitNode, template="Random::Execute(in Seed,in Minimum,in Maximum,in Duration,out Result)"}
   > RigVMDispatchNode(Element=@Random.Result, Array=RigVMVariableNode() {variable=strikes_speed_offset}.Value) {template="DISPATCH_RigVMDispatch_ArrayAdd(io Array,in Element,out Index)"}
   RigUnit_BeginExecution = &RigUnit_BeginExecution()
   RigVMFunction_Sequence = RigVMAggregateNode(ExecuteContext=@RigUnit_BeginExecution.ExecutePin)
   &RigUnit_SetTransform(
    Item=(Type=Bone,Name="pack")
    Value=(Rotation=(X=-0.5446,Y=0,Z=0,W=0.8387),Translation=(X=0,Y=-16,Z=-2),Scale3D=(X=1,Y=1,Z=1))
    bInitial=False
    Weight=&RigVMFunction_MathDoubleRemap(SourceMaximum=-50, SourceMinimum=-10, TargetMaximum=1, TargetMinimum=0, Value=RigVMVariableNode() {variable=Spine_rotation}.Value).Result
    ExecutePin=@RigVMFunction_Sequence.A
   )
   > &RigUnit_OffsetTransformForItem(
    Item=(Type=Bone,Name="pack")
    OffsetTransform=(Rotation=(X=0,Y=0,Z=0,W=1),Translation=(X=0,Y=0,Z=0),Scale3D=(X=1,Y=1,Z=1))
    Weight=1
    bPropagateToChildren=True
    OffsetTransform.Translation=RigVMVariableNode() {variable=pack_start_pos}.Value
   )
   RigUnit_ItemArray = &RigUnit_ItemArray(Items="((Type=Bone,Name=\"strike_1\"),(Type=Bone,Name=\"strike_2\"),(Type=Bone,Name=\"strike_3\"),(Type=Bone,Name=\"strike_4\"),(Type=Bone,Name=\"strike_5\"),(Type=Bone,Name=\"strike_6\"),(Type=Bone,Name=\"strike_7\"),(Type=Bone,Name=\"strike_8\"),(Type=Bone,Name=\"strike_9\"),(Ty ...(truncated)")
   For_Each = RigVMDispatchNode(Array=@RigUnit_ItemArray.Items, ExecuteContext=@RigVMFunction_Sequence.C) {template="DISPATCH_RigVMDispatch_ArrayIterator(in Array,out Element,out Index,out Count,out Ratio)"}
   n2 = RigVMDispatchNode(Index=@For_Each.Index, Array=RigVMVariableNode() {variable=strikes_speed_offset}.Value) {template="DISPATCH_RigVMDispatch_ArrayGetAtIndex(in Array,in Index,out Element)"}
   n3 = RigVMVariableNode() {variable=time_from_start}
   n4 = RigVMDispatchNode(Index=@For_Each.Index, Array=RigVMVariableNode() {variable=strikes_moving_offset}.Value) {template="DISPATCH_RigVMDispatch_ArrayGetAtIndex(in Array,in Index,out Element)"}
   RigVMFunction_MathIntToDouble = &RigVMFunction_MathIntToFloat(Value=@For_Each.Index) {template="Cast::Execute(in Value,out Result)"}
   n5 = /Script/RigVM.RigVMFunction_MathDoubleSin(Value=&RigVMFunction_MathDoubleAdd(
     B=&RigVMFunction_MathDoubleMul(B=@n2.Element, A=@n3.Value).Result
     A=&RigVMFunction_MathDoubleMul(B=@n4.Element, A=@RigVMFunction_MathIntToDouble.Result).Result
    ).Result) {is=struct RigVMUnitNode, template="Sin::Execute(in Value,out Result)"}
   n6 = RigVMVariableNode() {variable=strikes_pos_offset}
   > Set Transform = /Script/ControlRig.RigUnit_SetTranslation(
    Space=LocalSpace
    Value=(X=0,Y=0,Z=0)
    Weight=1
    bInitial=False
    bPropagateToChildren=true
    Item=@For_Each.Element
    Value.X=&RigVMFunction_MathDoubleMul(A=&RigVMFunction_MathDoubleAdd(A=@n5.Result, B=@n6.Value).Result, B=RigVMVariableNode() {variable=strikes_move_dist}.Value).Result
   ) {is=struct RigVMUnitNode, template="Set Transform::Execute(in Item,in Space,in bInitial,in Value,in Weight,in bPropagateToChildren)"}
   RigVMVariableNode() {variable=strikes_move_speed}
  ]}
]
hierarchy [
 root {loc=<0 0 0>, rot=<0 0 0 1>}
 MSTR_pack {parent=root, loc=<0 -0.7316 0>, rot=<0 0 0 1>}
 pack {parent=MSTR_pack, loc=<0 0 0>, rot=<0 0 0 1>}
 MSTR_strike_1 {parent=pack, loc=<0.0151 -1.0369 12.4509>, rot=<0 0.7071 0 -0.7071>}
 strike_1 {parent=MSTR_strike_1, loc=<0 0 0>, rot=<0 0 0 1>}
 MSTR_strike_2 {parent=pack, loc=<-6.2851 -1.0223 11.2571>, rot=<0 -0.8994 0 0.4372>}
 strike_2 {parent=MSTR_strike_2, loc=<0 0 0>, rot=<0 0 0 1>}
 MSTR_strike_3 {parent=pack, loc=<-10.342 -0.9564 7.2056>, rot=<0 -0.9705 0 0.2409>}
 strike_3 {parent=MSTR_strike_3, loc=<0 0 0>, rot=<0 0 0 1>}
 MSTR_strike_4 {parent=pack, loc=<-12.7562 -0.9564 1.9714>, rot=<0 -0.9992 0 0.0394>}
 strike_4 {parent=MSTR_strike_4, loc=<0 0 0>, rot=<0 0 0 1>}
 MSTR_strike_5 {parent=pack, loc=<-10.797 -0.9564 -5.742>, rot=<0 -0.9794 0 -0.202>}
 strike_5 {parent=MSTR_strike_5, loc=<0 0 0>, rot=<0 0 0 1>}
 MSTR_strike_6 {parent=pack, loc=<-7.8257 -1.0418 -9.0311>, rot=<0 -0.9508 0 -0.3098>}
 strike_6 {parent=MSTR_strike_6, loc=<0 0 0>, rot=<0 0 0 1>}
 MSTR_strike_7 {parent=pack, loc=<-4.1634 -1.0052 -11.5796>, rot=<0 -0.8814 0 -0.4724>}
 strike_7 {parent=MSTR_strike_7, loc=<0 0 0>, rot=<0 0 0 1>}
 MSTR_strike_8 {parent=pack, loc=<0.0151 -1.0369 -12.4509>, rot=<0 -0.7068 0 -0.7075>}
 strike_8 {parent=MSTR_strike_8, loc=<0 0 0>, rot=<0 0 0 1>}
 MSTR_strike_9 {parent=pack, loc=<4.1634 -1.0052 -11.5796>, rot=<0 -0.4714 0 -0.8819>}
 strike_9 {parent=MSTR_strike_9, loc=<0 0 0>, rot=<0 0 0 1>}
 MSTR_strike_10 {parent=pack, loc=<7.8257 -1.0418 -9.0311>, rot=<0 -0.3249 0 -0.9457>}
 strike_10 {parent=MSTR_strike_10, loc=<0 0 0>, rot=<0 0 0 1>}
 MSTR_strike_11 {parent=pack, loc=<10.633 -0.9891 -5.6702>, rot=<0 -0.1967 0 -0.9805>}
 strike_11 {parent=MSTR_strike_11, loc=<0 0 0>, rot=<0 0 0 1>}
 MSTR_strike_12 {parent=pack, loc=<12.666 -1.0278 1.9511>, rot=<0 0.0414 0 -0.9991>}
 strike_12 {parent=MSTR_strike_12, loc=<0 0 0>, rot=<0 0 0 1>}
 MSTR_strike_13 {parent=pack, loc=<10.2843 -0.9992 7.161>, rot=<0 0.2516 0 -0.9678>}
 strike_13 {parent=MSTR_strike_13, loc=<0 0 0>, rot=<0 0 0 1>}
 MSTR_strike_14 {parent=pack, loc=<6.1786 -1.0473 11.142>, rot=<0 0.453 0 -0.8915>}
 strike_14 {parent=MSTR_strike_14, loc=<0 0 0>, rot=<0 0 0 1>}
]
variables [
 Spine_rotation {type=double, default=0}
 pack_startPos_offset {type=FVector, default=(X=0,Y=0,Z=0)}
 pack_start_pos {type=FVector, default=(X=0,Y=0,Z=0)}
 strikes_move_dist {type=double, default=0}
 strikes_move_speed {type=double, default=0}
 strikes_moving_offset {type="TArray<double>"}
 strikes_pos_offset {type=double, default=0}
 strikes_speed_offset {type="TArray<double>"}
 time_from_start {type=double, default=0}
]
