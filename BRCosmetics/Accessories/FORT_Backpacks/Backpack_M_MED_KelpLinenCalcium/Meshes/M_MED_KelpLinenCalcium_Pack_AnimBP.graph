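# Backpack AnimBlueprint for the KelpLinenCalcium pack that animates it through a Control Rig and adjusts its placement to the character's state.
# Each update feeds game time and the owner mesh's spine_05-to-spine_01 rotation delta (yaw, stored as spine_Pitch) into the rig.
# A constant MSTR_pack offset precedes the rig.
# Afterward, boolean-blended additive bone transforms shift the pack during jumping, moving, emoting or skydiving, and roll it while skydiving.
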AnimBlueprint /BRCosmetics/Accessories/FORT_Backpacks/Backpack_M_MED_KelpLinenCalcium/Meshes/M_MED_KelpLinenCalcium_Pack_AnimBP
target_skeleton=&/M_MED_KelpLinenCalcium_Pack_Skeleton
graphs [
 AnimGraph {nodes [
   n1 = "Transform (Modify) Bone - Bone: MSTR_pack"(ComponentPose=Local To Component(LocalPose=Input Pose()), Alpha=1, Translation="0,2,0", Rotation="0,0,0", Scale="1,1,1") {
    translationMode=BMM_Additive
    translationSpace=BCS_BoneSpace
    alphaBoolBlend {blendOption=Linear}
    boneToModify {boneName=MSTR_pack}
    translation {y=2}
   }
   n2 = Control Rig!(
    Source=Component To Local(ComponentPose=@n1)
    Alpha
    time_from_start:double=Get time_from_start() -> (time_from_start:double)
    Spine_rotation:double=Get spine_Pitch() -> (spine_Pitch:double)
   ) -> (Pose) {
    alphaBoolBlend {blendOption=Linear}
    controlRigAssetReference {blueprintRigClass=&/M_MED_KelpLinenCalcium_Pack_CtrlRig.M_MED_KelpLinenCalcium_Pack_CtrlRig_C}
   }
   n3 = "Transform (Modify) Bone - Bone: pack"!(
    ComponentPose=Local To Component(LocalPose=@n2)
    bAlphaBoolEnabled:bool=Get BackpackOffset() -> (BackpackOffset:bool)
    Translation="0,-2,0"
    Rotation="0,0,0"
    Scale
   ) -> (Pose) {
    alphaInputType=Bool
    rotationMode=BMM_Additive
    rotationSpace=BCS_BoneSpace
    translationMode=BMM_Additive
    translationSpace=BCS_BoneSpace
    alphaBoolBlend {blendInTime=0.3, blendOption=Custom, blendOutTime=0.1}
    boneToModify {boneName=pack}
    translation {y=-2}
   }
   n4 = "Transform (Modify) Bone - Bone: pack"(ComponentPose=@n3, bAlphaBoolEnabled=Get bIsSkyDiving(), Translation="0,0,0", Rotation="0,0,5") {
    alphaInputType=Bool
    rotationMode=BMM_Additive
    rotationSpace=BCS_BoneSpace
    translationMode=BMM_Additive
    translationSpace=BCS_BoneSpace
    alphaBoolBlend {blendInTime=0.3, blendOption=Custom, blendOutTime=0.1}
    boneToModify {boneName=pack}
    rotation {roll=5}
   }
   Output Pose(Result=Component To Local(ComponentPose=@n4)) {layerGroup=None}
  ]}
 EventGraph {nodes [
   Event BlueprintUpdateAnimation()
   > Set time_from_start(time_from_start:double=GetGameTimeInSeconds()) -> (Output_Get:double)
   n1 = GetOwnerSkeletalMeshComponent()
   > Set spine_Pitch(spine_Pitch:double=BreakRotator(InRot=GetDeltaTransformFromRefPose!(self=@n1, BoneName=spine_05, BaseName=spine_01) -> (ReturnValue_Location:Vector, ReturnValue_Rotation:Rotator, ReturnValue_Scale:Vector).ReturnValue_Rotation).Yaw) -> (Output_Get:double)
   > Set BackpackOffset(BackpackOffset:bool=OR Boolean!(A=Get bIsJumping(), B=Get bIsMoving2D(), C:bool=Get bIsPlayingEmote(), D:bool=Get bIsSkyDiving()) -> (ReturnValue)) -> (Output_Get:bool)
  ]}
]
variables [
 BackpackOffset {type=boolean}
 char_velocity {type=Vector}
 spine_Pitch {type=number}
 time_from_start {type=number}
]
